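# Control Rig for the Parcel Prank Surprise back bling that corrects scaling on the spring puppet's gun barrels and pop-out flags.
# Each barrel's local X scale is clamp-remapped and applied back onto that barrel as an additive local scale.
# Each flag bone's local X scale is likewise remapped to drive Z scale on its child flag bone.
# Four chained ModifyBoneTransforms units do the work, without controls or variables.
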ControlRigBlueprint /BRCosmetics/Accessories/FORT_Backpacks/M_MED_Parcel_Prank_Surprise_Backpack/Meshes/M_MED_Parcel_Prank_Surprise_Pack_CtrlRig/Prank_gun
graphs [
 RigVMFunctionLibrary {links [], nodes []}
 RigVMModel {nodes [
   RigUnit_BeginExecution = &RigUnit_BeginExecution()
   /Script/ControlRig.RigUnit_ModifyBoneTransforms(
    BoneToModify=((Bone="dyn_l_gun_barrel",Transform=(Rotation=(X=0,Y=0,Z=0,W=1),Translation=(X=0,Y=0,Z=0),Scale3D=(X=1,Y=1,Z=1))))
    Mode=AdditiveLocal
    Weight=1
    WeightMaximum=1
    WeightMinimum=0
    BoneToModify.0.Transform.Scale3D.X=&RigVMFunction_MathFloatRemap(
     SourceMaximum=1
     SourceMinimum=0
     TargetMinimum=0.05
     bClamp=true
     Value=&RigUnit_GetBoneTransform(Bone=dyn_l_gun_barrel, Space=LocalSpace).Transform.Scale3D.X
    ).Result
    ExecutePin=@RigUnit_BeginExecution.ExecutePin
   ) {is=struct RigVMUnitNode}
   > &RigUnit_ModifyBoneTransforms(
    BoneToModify=((Bone="dyn_r_gun_barrel",Transform=(Rotation=(X=0,Y=0,Z=0,W=1),Translation=(X=0,Y=0,Z=0),Scale3D=(X=1,Y=1,Z=1))))
    Mode=AdditiveLocal
    Weight=1
    WeightMaximum=1
    WeightMinimum=0
    BoneToModify.0.Transform.Scale3D.X=&RigVMFunction_MathFloatRemap(
     SourceMaximum=1
     SourceMinimum=0
     TargetMinimum=0.05
     bClamp=true
     Value=&RigUnit_GetBoneTransform(Bone=dyn_r_gun_barrel, Space=LocalSpace).Transform.Scale3D.X
    ).Result
   )
   > RigVMStructNode_0 = &RigUnit_ModifyBoneTransforms(
    BoneToModify=((Bone="dyn_r_flag_01",Transform=(Rotation=(X=0,Y=0,Z=0,W=1),Translation=(X=0,Y=0,Z=0),Scale3D=(X=1,Y=1,Z=1))))
    Mode=AdditiveLocal
    Weight=1
    WeightMaximum=1
    WeightMinimum=0
    BoneToModify.0.Transform.Scale3D.Z=&RigVMFunction_MathFloatRemap(
     SourceMaximum=1
     SourceMinimum=0.1
     TargetMinimum=0.1
     bClamp=true
     Value=&RigUnit_GetBoneTransform(Bone=dyn_r_flag, Space=LocalSpace).Transform.Scale3D.X
    ).Result
   )
   > RigVMStructNode_0_1 = &RigUnit_ModifyBoneTransforms(
    BoneToModify=((Bone="dyn_l_flag_01",Transform=(Rotation=(X=0,Y=0,Z=0,W=1),Translation=(X=0,Y=0,Z=0),Scale3D=(X=1,Y=1,Z=1))))
    Mode=AdditiveLocal
    Weight=1
    WeightMaximum=1
    WeightMinimum=0
    BoneToModify.0.Transform.Scale3D.Z=&RigVMFunction_MathFloatRemap(
     SourceMaximum=1
     SourceMinimum=0.1
     TargetMinimum=0.1
     bClamp=true
     Value=&RigUnit_GetBoneTransform(Bone=dyn_l_flag, Space=LocalSpace).Transform.Scale3D.X
    ).Result
   )
  ]}
]
hierarchy [
 root {loc=<0 0 0>, rot=<0 0 0 1>}
 pack {parent=root, loc=<0 -0.7316 0>, rot=<0 0 0 1>}
 dyn_lid_MSTR {parent=pack, loc=<-0.0029 -1.5257 2.4377>, rot=<0.1802 0.1802 0.6838 -0.6838>}
 dyn_lid {parent=dyn_lid_MSTR, loc=<0 0 0>, rot=<0 0 0 1>}
 dyn_handle_MSTR {parent=pack, loc=<-7.0645 -6.4709 -5.5896>, rot=<0 0 1 0>}
 dyn_handle {parent=dyn_handle_MSTR, loc=<0 0 0>, rot=<0 0 0 1>}
 dyn_spring_MSTR {parent=pack, loc=<-0.0029 -4.8632 -7.1012>, rot=<0.4217 0.4217 0.5676 -0.5676>}
 dyn_spring_01 {parent=dyn_spring_MSTR, loc=<0.943 0 0>, rot=<0 0 0 -1>}
 dyn_spring_02 {parent=dyn_spring_01, loc=<2.6193 0 0.0655>, rot=<0 0.0135 0 -0.9999>}
 dyn_spring_03 {parent=dyn_spring_02, loc=<2.7839 0 -0.0668>, rot=<0 -0.0513 0 -0.9987>}
 dyn_head {parent=dyn_spring_03, loc=<8.2425 0 0>, rot=<0 -0.0067 0 -1>}
 dyn_bowtie_l {parent=dyn_head, loc=<-0.4047 -2.7115 -1.023>, rot=<0.2658 -0.5333 0.5267 -0.6062>}
 dyn_bowtie_r {parent=dyn_head, loc=<-0.4047 2.6278 -1.023>, rot=<-0.2111 -0.5959 -0.4856 -0.6038>}
 dyn_l_arm {parent=dyn_spring_03, loc=<7.1996 -1.7731 0.3166>, rot=<0.2409 -0.5379 0.6551 -0.4728>}
 dyn_l_hand {parent=dyn_l_arm, loc=<5.7315 -0.0781 -0.0733>, rot=<0.0048 0.0085 -0.009 -0.9999>}
 dyn_l_gun_barrel {parent=dyn_l_hand, loc=<1.3457 1.166 3.7933>, rot=<0.0196 0.3441 -0.1571 -0.9255>}
 dyn_l_flag {parent=dyn_l_gun_barrel, loc=<4.8701 0.0278 0.0589>, rot=<0.0346 -0.7132 0.0404 -0.6989>}
 dyn_l_flag_01 {parent=dyn_l_flag, loc=<8.7937 0 0>, rot=<0 0 0 1>}
 dyn_r_arm {parent=dyn_spring_03, loc=<7.1996 1.9134 0.3166>, rot=<-0.3587 -0.4391 -0.6102 -0.5534>}
 dyn_r_hand {parent=dyn_r_arm, loc=<5.1904 -0.0141 -0.1007>, rot=<0.0002 0.0088 -0.0012 -1>}
 dyn_r_gun_barrel {parent=dyn_r_hand, loc=<1.8821 -0.5105 3.4854>, rot=<0.0079 0.2719 0.0509 -0.9609>}
 dyn_r_flag {parent=dyn_r_gun_barrel, loc=<4.9409 -0.0842 0.0042>, rot=<0.0006 -0.7144 -0.0014 -0.6997>}
 dyn_r_flag_01 {parent=dyn_r_flag, loc=<8.5428 0 0>, rot=<0 0 0 1>}
 Parcel_in {type=Curve, loc=<0 0 0>, rot=<0 0 0 1>}
 Parcel_out {type=Curve, loc=<0 0 0>, rot=<0 0 0 1>}
 Pose_0 {type=Curve, loc=<0 0 0>, rot=<0 0 0 1>}
 Parcel_mid {type=Curve, loc=<0 0 0>, rot=<0 0 0 1>}
 Parcel_out_2 {type=Curve, loc=<0 0 0>, rot=<0 0 0 1>}
 Pose_1 {type=Curve, loc=<0 0 0>, rot=<0 0 0 1>}
 Pose_2 {type=Curve, loc=<0 0 0>, rot=<0 0 0 1>}
 Pose_3 {type=Curve, loc=<0 0 0>, rot=<0 0 0 1>}
 Pose_4 {type=Curve, loc=<0 0 0>, rot=<0 0 0 1>}
 Pose_5 {type=Curve, loc=<0 0 0>, rot=<0 0 0 1>}
 Pose_6 {type=Curve, loc=<0 0 0>, rot=<0 0 0 1>}
 D__build_FortressArtSource_assets_Backpack_Male_Medium_Parcel_Prank_Surprise_Rig {type=Curve, loc=<0 0 0>, rot=<0 0 0 1>}
 Surprise_in {type=Curve, loc=<0 0 0>, rot=<0 0 0 1>}
 Surprise_out {type=Curve, loc=<0 0 0>, rot=<0 0 0 1>}
 Breakdown_2 {type=Curve, loc=<0 0 0>, rot=<0 0 0 1>}
 Breakdown_1 {type=Curve, loc=<0 0 0>, rot=<0 0 0 1>}
 Breakdown_3 {type=Curve, loc=<0 0 0>, rot=<0 0 0 1>}
]
