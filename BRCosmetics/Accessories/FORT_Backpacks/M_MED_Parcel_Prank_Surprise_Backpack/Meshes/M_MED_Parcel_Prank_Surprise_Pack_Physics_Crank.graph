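# Physics asset for the Parcel Prank Surprise backpack cosmetic, set up for its crank handle.
# A sphere body on the dyn_handle bone and a box body on the pack bone are joined by a single constraint.
# The constraint locks both swing axes and leaves twist free, so the handle acts as a hinge that rotates about one axis, with its 45-degree twist limit not enforced.
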PhysicsAsset /BRCosmetics/Accessories/FORT_Backpacks/M_MED_Parcel_Prank_Surprise_Backpack/Meshes/M_MED_Parcel_Prank_Surprise_Pack_Physics_Crank
bodies {
 dyn_handle [{shapeType=Sphere, radius=7.7244, center=<5.2628 3.6103 -3.1219>}]
 pack [{shapeType=Box, extentX=8.0813, extentY=8.2332, extentZ=8.5991, center=<0.3648 -7.3382 -5.2816>, rotation=<rot 0 93.9198 -86.3904>}]
}
constraints [{
  bone1Name=dyn_handle
  bone2Name=pack
  swing1LimitDegrees=0
  swing1Motion=Locked
  swing2LimitDegrees=0
  swing2Motion=Locked
  twistLimitDegrees=45
  twistMotion=Free
 }]
