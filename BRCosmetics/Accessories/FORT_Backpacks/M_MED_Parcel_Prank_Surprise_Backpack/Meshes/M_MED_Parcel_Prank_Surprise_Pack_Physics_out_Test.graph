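# Physics asset for the Parcel Prank Surprise backpack cosmetic, a jack-in-the-box style rig with a lid, spring-mounted head, arms and flags.
# The pack locks rotationally to a three-box root body, and the lid hinges from it on the twist axis.
# A three-segment spring chain rises from the pack with narrow swing and near-zero twist limits, carrying head and arms.
# Each flag spins freely about its arm's twist axis with swing locked.
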PhysicsAsset /BRCosmetics/Accessories/FORT_Backpacks/M_MED_Parcel_Prank_Surprise_Backpack/Meshes/M_MED_Parcel_Prank_Surprise_Pack_Physics_out_Test
bodies {
 dyn_head [{shapeType=Capsule, length=5.787, radius=7.4913, center=<6.4855 0.0029 1.8746>, rotation=<rot 0 0 -90>}]
 dyn_l_arm [{shapeType=Capsule, length=11.4541, radius=2.9095, center=<6.6208 1.4645 3.8408>, rotation=<rot -11.0913 100.3119 -49.6918>}]
 dyn_l_flag [{shapeType=Box, extentX=0.7777, extentY=10.7575, extentZ=5.2461, center=<3.5585 -0.0345 0.1223>, rotation=<rot 0.4296 -90 -1.6622>}]
 dyn_lid [{shapeType=Box, extentX=1.8922, extentY=15.6122, extentZ=15.7484, center=<0.2577 0.0028 5.2611>, rotation=<rot -5 0 0>}]
 dyn_r_arm [{shapeType=Capsule, length=12.9629, radius=2.2314, center=<6.8878 -0.7489 2.7617>, rotation=<rot 2.5129 82.418 -54.9069>}]
 dyn_r_flag [{shapeType=Box, extentX=0.7648, extentY=9.9445, extentZ=5.2172, center=<4.4697 0.1065 0.0404>, rotation=<rot 0.1285 -89.9995 -1.2962>}]
 dyn_spring_01 [{shapeType=Capsule, length=1.9469, radius=1.8557, center=<0.9343 -0.0121 0.1173>, rotation=<rot 0 0 -90>}]
 dyn_spring_02 [{shapeType=Capsule, length=1.9616, radius=1.9038, center=<0.4621 -0.0344 0.0707>, rotation=<rot 0 0 -90>}]
 dyn_spring_03 [{shapeType=Capsule, length=5.0271, radius=4.3523, center=<3.6355 0.0029 0.2328>, rotation=<rot 90 0 0>}]
 pack [{shapeType=Box, extentX=23.6976, extentY=18.33, extentZ=8.8653, center=<0.1593 -7.1915 -2.0911>, rotation=<rot 0 0 -15.0103>}]
 root [
  {shapeType=Box, extentX=43.7408, extentY=4.0169, extentZ=51.442, center=<0 3.0714 -0.5285>, rotation=<rot 0 0 10.007>}
  {shapeType=Box, extentX=19.667, extentY=7.5, extentZ=39.1535, center=<-18.6643 -0.8361 7.0789>}
  {shapeType=Box, extentX=18.6357, extentY=7.5, extentZ=38.8001, center=<18.236 0 7.0431>}
 ]
}
constraints [
 {bone1Name=dyn_spring_02, bone2Name=dyn_spring_01, swing1LimitDegrees=20, swing2LimitDegrees=20, twistLimitDegrees=1}
 {bone1Name=dyn_spring_03, bone2Name=dyn_spring_02, swing1LimitDegrees=25, swing2LimitDegrees=20, twistLimitDegrees=1}
 {bone1Name=dyn_head, bone2Name=dyn_spring_03, swing1LimitDegrees=25, swing2LimitDegrees=30, twistLimitDegrees=1}
 {bone1Name=dyn_spring_01, bone2Name=pack, swing1LimitDegrees=10, swing2LimitDegrees=10, twistLimitDegrees=1}
 {
  bone1Name=dyn_lid
  bone2Name=pack
  swing1LimitDegrees=0
  swing1Motion=Locked
  swing2LimitDegrees=0
  swing2Motion=Locked
  twistLimitDegrees=45
 }
 {
  bone1Name=pack
  bone2Name=root
  swing1LimitDegrees=0
  swing1Motion=Locked
  swing2LimitDegrees=0
  swing2Motion=Locked
  twistLimitDegrees=0
  twistMotion=Locked
 }
 {bone1Name=dyn_l_arm, bone2Name=dyn_spring_03, swing1LimitDegrees=20, swing2LimitDegrees=20, twistLimitDegrees=1}
 {bone1Name=dyn_r_arm, bone2Name=dyn_spring_03, swing1LimitDegrees=20, swing2LimitDegrees=20, twistLimitDegrees=1}
 {
  bone1Name=dyn_r_flag
  bone2Name=dyn_r_arm
  swing1LimitDegrees=45
  swing1Motion=Locked
  swing2LimitDegrees=45
  swing2Motion=Locked
  twistLimitDegrees=45
  twistMotion=Free
 }
 {
  bone1Name=dyn_l_flag
  bone2Name=dyn_l_arm
  swing1LimitDegrees=0
  swing1Motion=Locked
  swing2LimitDegrees=0
  swing2Motion=Locked
  twistLimitDegrees=45
  twistMotion=Free
 }
]
