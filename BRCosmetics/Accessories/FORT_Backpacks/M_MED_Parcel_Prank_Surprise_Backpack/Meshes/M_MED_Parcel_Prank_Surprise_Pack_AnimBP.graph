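# Animation Blueprint for the Parcel Prank Surprise backpack, a jack-in-the-box that cranks after a player kill, then pops out.
# Watched kill stat spins the crank for a random duration, then frame-counted timers stage pop-out, physics, barrels and flag scaling before resetting.
# Bool blends layer poses built from RigidBody simulation, bone offsets and Control Rigs, with random force jitter and PrankFX dispatchers.
# The Prank_flag rig branch never reaches the output pose.
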AnimBlueprint /BRCosmetics/Accessories/FORT_Backpacks/M_MED_Parcel_Prank_Surprise_Backpack/Meshes/M_MED_Parcel_Prank_Surprise_Pack_AnimBP
target_skeleton=&/M_MED_Parcel_Prank_Surprise_Pack_Skeleton
graphs [
 AnimGraph {nodes [
   n1 = Control Rig(Source=Use cached pose 'AnimCache'()) {
    alphaBoolBlend {blendOption=Linear}
    controlRigAssetReference {blueprintRigClass=/BRCosmetics/Accessories/FORT_Backpacks/M_MED_Parcel_Prank_Surprise_Backpack/Meshes/M_MED_Parcel_Prank_Surprise_Pack_CtrlRig/Prank_gun.Prank_gun_C}
   }
   n2 = RigidBody(ComponentPose=Local To Component(LocalPose=@n1)) {
    bClampLinearTranslationLimitToRefPose=true
    bOverrideWorldGravity=true
    overridePhysicsAsset=&/M_MED_Parcel_Prank_Surprise_Pack_Physics_out_Test
    simulationSpace=BaseBoneSpace
    alphaBoolBlend {blendOption=Linear}
    baseBoneRef {boneName=root}
    componentLinearAccScale=<0.5 0.6 0.7>
    componentLinearVelScale=<0.3 0.6 0.5>
    overrideWorldGravity {z=-980}
   }
   n3 = RigidBody(ComponentPose=@n2) {
    bClampLinearTranslationLimitToRefPose=true
    bOverrideWorldGravity=true
    overridePhysicsAsset=&/M_MED_Parcel_Prank_Surprise_Pack_Physics_Crank
    alphaBoolBlend {blendOption=Linear}
    componentLinearAccScale=<0.3 0.8 0.8>
    componentLinearVelScale=<0.3 0.7 0.7>
    overrideWorldGravity {z=-980}
   }
   n4 = Get Flag Scale() -> (Flag Scale:Vector)
   n5 = "Transform (Modify) Bone - Bone: dyn_r_arm"(
    ComponentPose="Transform (Modify) Bone - Bone: dyn_l_flag"(
     ComponentPose="Transform (Modify) Bone - Bone: dyn_r_flag"(ComponentPose=@n3, Alpha=1, Translation="0,0,0", Rotation="0,0,0", Scale=@n4) {scaleMode=BMM_Replace, alphaBoolBlend {blendOption=Linear}, boneToModify {boneName=dyn_r_flag}}
     Alpha=1
     Translation="0,0,0"
     Rotation="0,0,0"
     Scale=Get Flag 2 scale() -> (Flag 2 scale:Vector)
    ) {scaleMode=BMM_Replace, alphaBoolBlend {blendOption=Linear}, boneToModify {boneName=dyn_l_flag}}
    Alpha=1
    Translation="0,0,0"
    Rotation="3,0.000,0.000"
    Scale="1,1,1"
   ) {
    rotationMode=BMM_Additive
    scaleSpace=BCS_BoneSpace
    alphaBoolBlend {blendOption=Linear}
    boneToModify {boneName=dyn_r_arm}
    rotation {pitch=3}
   }
   n6 = "Transform (Modify) Bone - Bone: dyn_l_arm"(ComponentPose=@n5, Alpha=1, Translation="0,0,0", Rotation="-3,0.000,0.000", Scale="1,1,1") {
    rotationMode=BMM_Additive
    scaleSpace=BCS_BoneSpace
    alphaBoolBlend {blendOption=Linear}
    boneToModify {boneName=dyn_l_arm}
    rotation {pitch=-3}
   }
   n7 = "Transform (Modify) Bone - Bone: dyn_spring_MSTR"(ComponentPose=@n6, Alpha=1, Translation="0,0,2", Rotation="0,0,0", Scale="1,1,1") {translationMode=BMM_Additive, alphaBoolBlend {blendOption=Linear}, boneToModify {boneName=dyn_spring_MSTR}, translation {z=2}}
   n8 = "Transform (Modify) Bone - Bone: dyn_spring_MSTR"!(ComponentPose=@n7, bAlphaBoolEnabled:bool=Get bIsTargeting(), Translation="0,0,0", Rotation="0,-45,0", Scale="1,1,1") -> (Pose) {
    alphaInputType=Bool
    rotationMode=BMM_Additive
    alphaBoolBlend {blendOption=Linear}
    boneToModify {boneName=dyn_spring_MSTR}
    rotation {yaw=-45}
   }
   n9 = RigidBody(ComponentPose=Local To Component(LocalPose=Use cached pose 'AnimCache'()), ExternalForce=Get Surprise Force() -> (Surprise Force:Vector)) {
    bClampLinearTranslationLimitToRefPose=true
    bOverrideWorldGravity=true
    overridePhysicsAsset=&M_MED_Parcel_Prank_Surprise_Pack_Physics_out_Test
    simulationSpace=BaseBoneSpace
    alphaBoolBlend {blendOption=Linear}
    baseBoneRef {boneName=root}
    componentLinearAccScale=<0.7 0.7 0.7>
    componentLinearVelScale=<0.7 0.6 0.5>
    overrideWorldGravity {z=-980}
   }
   n10 = RigidBody(ComponentPose=@n9) {
    bClampLinearTranslationLimitToRefPose=true
    bOverrideWorldGravity=true
    overridePhysicsAsset=&M_MED_Parcel_Prank_Surprise_Pack_Physics_Crank
    alphaBoolBlend {blendOption=Linear}
    componentLinearAccScale=<0.7 0.7 0.7>
    componentLinearVelScale=<0.5 0.5 0.5>
    overrideWorldGravity {z=-980}
   }
   n11 = "Transform (Modify) Bone - Bone: dyn_spring_03"(ComponentPose=@n10, Alpha=1, Translation="0,0,-2", Rotation=Get Spring_angle() -> (Spring_angle:Rotator), Scale="1,1,1") {translationMode=BMM_Additive, alphaBoolBlend {blendOption=Linear}, boneToModify {boneName=dyn_spring_03}, translation {z=-2}}
   n12 = "Transform (Modify) Bone - Bone: dyn_r_arm"(ComponentPose=@n11, Alpha=1, Translation="0,0,0", Rotation="40,0,0", Scale="1,1,1") {
    rotationMode=BMM_Additive
    scaleSpace=BCS_BoneSpace
    alphaBoolBlend {blendOption=Linear}
    boneToModify {boneName=dyn_r_arm}
    rotation {pitch=40}
   }
   n13 = "Transform (Modify) Bone - Bone: dyn_l_arm"(ComponentPose=@n12, Alpha=1, Translation="0,0,0", Rotation="-40,0,0", Scale="1,1,1") {
    rotationMode=BMM_Additive
    scaleSpace=BCS_BoneSpace
    alphaBoolBlend {blendOption=Linear}
    boneToModify {boneName=dyn_l_arm}
    rotation {pitch=-40}
   }
   n14 = "Transform (Modify) Bone - Bone: dyn_head"(ComponentPose=@n13, Alpha=1, Translation="0,0,0", Rotation="0,0,-20", Scale="1,1,1") {
    rotationMode=BMM_Additive
    translationMode=BMM_Additive
    translationSpace=BCS_ParentBoneSpace
    alphaBoolBlend {blendOption=Linear}
    boneToModify {boneName=dyn_head}
    rotation {roll=-20}
   }
   n15 = "Transform (Modify) Bone - Bone: dyn_r_gun_barrel"(
    ComponentPose="Transform (Modify) Bone - Bone: dyn_r_flag"(
     ComponentPose="Transform (Modify) Bone - Bone: dyn_l_flag"(ComponentPose=@n14, Alpha=1, Translation="0,0,0", Rotation="0,0,0", Scale="0.05,1,1") {scaleMode=BMM_Replace, alphaBoolBlend {blendOption=Linear}, boneToModify {boneName=dyn_l_flag}, scale {x=0.05}}
     Alpha=1
     Translation="0,0,0"
     Rotation="0,0,0"
     Scale="0.05,1,1"
    ) {scaleMode=BMM_Replace, alphaBoolBlend {blendOption=Linear}, boneToModify {boneName=dyn_r_flag}, scale {x=0.05}}
    Alpha=1
    Translation="0,0,0"
    Rotation="0,0,0"
    Scale="0.01,1,1"
   ) {scaleMode=BMM_Replace, alphaBoolBlend {blendOption=Linear}, boneToModify {boneName=dyn_r_gun_barrel}, scale {x=0.01}}
   n16 = "Transform (Modify) Bone - Bone: dyn_l_gun_barrel"(ComponentPose=@n15, Alpha=1, Translation="0,0,0", Rotation="0,0,0", Scale="0.01,1,1") {scaleMode=BMM_Replace, alphaBoolBlend {blendOption=Linear}, boneToModify {boneName=dyn_l_gun_barrel}, scale {x=0.01}}
   n17 = "Transform (Modify) Bone - Bone: dyn_l_flag_01"(ComponentPose=@n16, Alpha=1, Translation="0,0,0", Rotation="0,0,0", Scale="0,1,0") {scaleMode=BMM_Replace, alphaBoolBlend {blendOption=Linear}, boneToModify {boneName=dyn_l_flag_01}, scale {x=0, z=0}}
   n18 = "Transform (Modify) Bone - Bone: dyn_r_flag_01"(ComponentPose=@n17, Alpha=1, Translation="0,0,0", Rotation="0,0,0", Scale="0,1.000,0") {scaleMode=BMM_Replace, alphaBoolBlend {blendOption=Linear}, boneToModify {boneName=dyn_r_flag_01}, scale {x=0, z=0}}
   n19 = "Transform (Modify) Bone - Bone: dyn_spring_MSTR"(ComponentPose=@n18, bAlphaBoolEnabled=Get bIsTargeting(), Translation="0,0,0", Rotation="0,-45,0", Scale="1,1,1") {
    alphaInputType=Bool
    rotationMode=BMM_Additive
    alphaBoolBlend {blendOption=Linear}
    boneToModify {boneName=dyn_spring_MSTR}
    rotation {yaw=-45}
   }
   n20 = Component To Local(ComponentPose=@n19)
   n21 = RigidBody(ComponentPose=Local To Component(LocalPose=Use cached pose 'AnimCache'()), ExternalForce=Get Surprise Force()) {
    bClampLinearTranslationLimitToRefPose=true
    bOverrideWorldGravity=true
    overridePhysicsAsset=&M_MED_Parcel_Prank_Surprise_Pack_Physics_out_Test
    simulationSpace=BaseBoneSpace
    alphaBoolBlend {blendOption=Linear}
    baseBoneRef {boneName=root}
    componentLinearAccScale=<2 2 2>
    componentLinearVelScale=<2 2 2>
    overrideWorldGravity {z=-980}
   }
   n22 = RigidBody(ComponentPose=@n21) {
    bClampLinearTranslationLimitToRefPose=true
    bOverrideWorldGravity=true
    overridePhysicsAsset=&M_MED_Parcel_Prank_Surprise_Pack_Physics_Crank
    alphaBoolBlend {blendOption=Linear}
    componentLinearAccScale=<1 1 1>
    componentLinearVelScale=<1 1 1>
    overrideWorldGravity {z=-980}
   }
   n23 = "Transform (Modify) Bone - Bone: dyn_r_gun_barrel"(
    ComponentPose="Transform (Modify) Bone - Bone: dyn_r_flag"(
     ComponentPose="Transform (Modify) Bone - Bone: dyn_l_flag"(ComponentPose=@n22, Alpha=1, Translation="0,0,0", Rotation="0,0,0", Scale="0.05,1,1") {scaleMode=BMM_Replace, alphaBoolBlend {blendOption=Linear}, boneToModify {boneName=dyn_l_flag}, scale {x=0.05}}
     Alpha=1
     Translation="0,0,0"
     Rotation="0,0,0"
     Scale="0.05,1,1"
    ) {scaleMode=BMM_Replace, alphaBoolBlend {blendOption=Linear}, boneToModify {boneName=dyn_r_flag}, scale {x=0.05}}
    Alpha=1
    Translation="0,0,0"
    Rotation="0,0,0"
    Scale="0.01,1,1"
   ) {scaleMode=BMM_Replace, alphaBoolBlend {blendOption=Linear}, boneToModify {boneName=dyn_r_gun_barrel}, scale {x=0.01}}
   n24 = "Transform (Modify) Bone - Bone: dyn_l_gun_barrel"(ComponentPose=@n23, Alpha=1, Translation="0,0,0", Rotation="0,0,0", Scale="0.01,1,1") {scaleMode=BMM_Replace, alphaBoolBlend {blendOption=Linear}, boneToModify {boneName=dyn_l_gun_barrel}, scale {x=0.01}}
   n25 = "Transform (Modify) Bone - Bone: dyn_head"(ComponentPose=@n24, Alpha=1, Translation="0,0,0", Rotation="0,0,15", Scale="1,1,1") {
    rotationMode=BMM_Additive
    translationMode=BMM_Additive
    translationSpace=BCS_ParentBoneSpace
    alphaBoolBlend {blendOption=Linear}
    boneToModify {boneName=dyn_head}
    rotation {roll=15}
   }
   n26 = "Transform (Modify) Bone - Bone: dyn_r_arm"(ComponentPose=@n25, Alpha=1, Translation="0,0,0", Rotation="10.000,0.000,0.000", Scale="1,1,1") {
    rotationMode=BMM_Additive
    scaleSpace=BCS_BoneSpace
    alphaBoolBlend {blendOption=Linear}
    boneToModify {boneName=dyn_r_arm}
    rotation {pitch=10}
   }
   n27 = "Transform (Modify) Bone - Bone: dyn_l_arm"(ComponentPose=@n26, Alpha=1, Translation="0,0,0", Rotation="-25.000,0.000,0.000", Scale="1,1,1") {
    rotationMode=BMM_Additive
    scaleSpace=BCS_BoneSpace
    alphaBoolBlend {blendOption=Linear}
    boneToModify {boneName=dyn_l_arm}
    rotation {pitch=-25}
   }
   n28 = "Transform (Modify) Bone - Bone: dyn_spring_MSTR"(ComponentPose=@n27, bAlphaBoolEnabled=Get bIsTargeting(), Translation="0,0,0", Rotation="0,-45,0", Scale="1,1,1") {
    alphaInputType=Bool
    rotationMode=BMM_Additive
    alphaBoolBlend {blendOption=Linear}
    boneToModify {boneName=dyn_spring_MSTR}
    rotation {yaw=-45}
   }
   n29 = "Transform (Modify) Bone - Bone: dyn_lid_MSTR"(
    ComponentPose=Local To Component(LocalPose=Use cached pose 'AnimCache'())
    Alpha=1
    Translation="0,0,0"
    Rotation="0.000,0.000,-130.000"
    Scale="1,1,1"
   ) {rotationMode=BMM_Additive, alphaBoolBlend {blendOption=Linear}, boneToModify {boneName=dyn_lid_MSTR}, rotation {roll=-130}}
   n30 = "Transform (Modify) Bone - Bone: dyn_lid_MSTR"(ComponentPose=@n29, Alpha=1, Translation="0.000,-2.499,0.421", Rotation="0,0,0", Scale="1,1,1") {translationMode=BMM_Additive, alphaBoolBlend {blendOption=Linear}, boneToModify {boneName=dyn_lid_MSTR}, translation {y=-2.499, z=0.421}}
   n31 = "Transform (Modify) Bone - Bone: dyn_l_arm"(ComponentPose=@n30, Alpha=1, Translation="0,0,0", Rotation="-50,0.000,0", Scale="1,1,1") {rotationMode=BMM_Additive, alphaBoolBlend {blendOption=Linear}, boneToModify {boneName=dyn_l_arm}, rotation {pitch=-50}}
   n32 = "Transform (Modify) Bone - Bone: dyn_r_arm"(ComponentPose=@n31, Alpha=1, Translation="0,0,0", Rotation="40,0.000,0", Scale="1,1,1") {rotationMode=BMM_Additive, alphaBoolBlend {blendOption=Linear}, boneToModify {boneName=dyn_r_arm}, rotation {pitch=40}}
   n33 = "Transform (Modify) Bone - Bone: dyn_head"(ComponentPose=@n32, Alpha=1, Translation="0,0,0", Rotation="0,0,-50", Scale="1,1,1") {rotationMode=BMM_Additive, alphaBoolBlend {blendOption=Linear}, boneToModify {boneName=dyn_head}, rotation {roll=-50}}
   n34 = RigidBody(ComponentPose="Transform (Modify) Bone - Bone: dyn_spring_02"(ComponentPose=@n33, Alpha=1, Translation="0,0,0", Rotation=Get Spring_angle(), Scale="1,1,1") {rotationMode=BMM_Additive, alphaBoolBlend {blendOption=Linear}, boneToModify {boneName=dyn_spring_02}}) {
    bClampLinearTranslationLimitToRefPose=true
    bOverrideWorldGravity=true
    overridePhysicsAsset=&/M_MED_Parcel_Prank_Surprise_Pack_Physics_Crank_in
    alphaBoolBlend {blendOption=Linear}
    componentLinearAccScale=<0.7 0.7 0.7>
    componentLinearVelScale=<0.5 0.5 0.5>
    overrideWorldGravity {z=-980}
   }
   n35 = "Transform (Modify) Bone - Bone: dyn_r_gun_barrel"(
    ComponentPose="Transform (Modify) Bone - Bone: dyn_r_flag"(
     ComponentPose="Transform (Modify) Bone - Bone: dyn_l_flag"(ComponentPose=@n34, Alpha=1, Translation="0,0,0", Rotation="0,0,0", Scale="0.05,1,1") {scaleMode=BMM_Replace, alphaBoolBlend {blendOption=Linear}, boneToModify {boneName=dyn_l_flag}, scale {x=0.05}}
     Alpha=1
     Translation="0,0,0"
     Rotation="0,0,0"
     Scale="0.05,1,1"
    ) {scaleMode=BMM_Replace, alphaBoolBlend {blendOption=Linear}, boneToModify {boneName=dyn_r_flag}, scale {x=0.05}}
    Alpha=1
    Translation="0,0,0"
    Rotation="0,0,0"
    Scale="0.05,1,1"
   ) {scaleMode=BMM_Replace, alphaBoolBlend {blendOption=Linear}, boneToModify {boneName=dyn_r_gun_barrel}, scale {x=0.05}}
   n36 = "Transform (Modify) Bone - Bone: dyn_l_gun_barrel"(ComponentPose=@n35, Alpha=1, Translation="0,0,0", Rotation="0,0,0", Scale="0.05,1,1") {scaleMode=BMM_Replace, alphaBoolBlend {blendOption=Linear}, boneToModify {boneName=dyn_l_gun_barrel}, scale {x=0.05}}
   n37 = "Transform (Modify) Bone - Bone: dyn_handle_MSTR"(ComponentPose=@n36, Alpha=1, Translation="0,0,0", Rotation=Get Crank Rot() -> (Crank Rot:Rotator), Scale="1,1,1") {rotationMode=BMM_Additive, rotationSpace=BCS_BoneSpace, alphaBoolBlend {blendOption=Linear}, boneToModify {boneName=dyn_handle_MSTR}}
   n38 = "Transform (Modify) Bone - Bone: dyn_spring_MSTR"!(ComponentPose=@n37, Alpha:float=1, Translation="0,0,0", Rotation="0,0,0", Scale=Get Surprise_Scale() -> (Surprise_Scale:Vector)) -> (Pose) {scaleMode=BMM_Replace, scaleSpace=BCS_BoneSpace, alphaBoolBlend {blendOption=Linear}, boneToModify {boneName=dyn_spring_MSTR}}
   n39 = "Transform (Modify) Bone - Bone: dyn_r_arm"(ComponentPose=@n38, Alpha=1, Translation="0,0,0", Rotation="0,0,0", Scale=Get arm_scale() -> (arm_scale:Vector)) {scaleMode=BMM_Replace, scaleSpace=BCS_BoneSpace, alphaBoolBlend {blendOption=Linear}, boneToModify {boneName=dyn_r_arm}}
   n40 = "Transform (Modify) Bone - Bone: dyn_l_arm"(ComponentPose=@n39, Alpha=1, Translation="0,0,0", Rotation="0,0,0", Scale=Get arm_scale()) {scaleMode=BMM_Replace, scaleSpace=BCS_BoneSpace, alphaBoolBlend {blendOption=Linear}, boneToModify {boneName=dyn_l_arm}}
   n41 = "Transform (Modify) Bone - Bone: dyn_spring_MSTR"!(ComponentPose=@n40, bAlphaBoolEnabled:bool=Get bIsTargeting(), Translation="0,0,0", Rotation="0,-45,0", Scale="1,1,1") -> (Pose) {
    alphaInputType=Bool
    rotationMode=BMM_Additive
    alphaBoolBlend {blendOption=Linear}
    boneToModify {boneName=dyn_spring_MSTR}
    rotation {yaw=-45}
   }
   n42 = Control Rig(Source=Component To Local(ComponentPose=@n41)) {
    alphaBoolBlend {blendOption=Linear}
    controlRigAssetReference {blueprintRigClass=&/M_MED_Parcel_Prank_Surprise_Pack_CtrlRig.M_MED_Parcel_Prank_Surprise_Pack_CtrlRig_C}
   }
   n43 = Two Way Blend!(A=@n42, B=Component To Local(ComponentPose=@n28), bAlphaBoolEnabled:bool=Get Surprise_Out() -> (Surprise_Out:bool)) -> (Pose) {blendNode {alphaInputType=Bool, alphaBoolBlend {blendInTime=0.4, blendOption=ExpIn, blendOutTime=0.2}}}
   n44 = Get Physic out() -> (Physic out:bool)
   Output Pose(Result=Two Way Blend(
     A=Two Way Blend(A=@n43, B=@n20, bAlphaBoolEnabled=@n44) {blendNode {alphaInputType=Bool, alphaBoolBlend {blendInTime=0.3, blendOption=CircularInOut, blendOutTime=0.3}}}
     B=Component To Local(ComponentPose=@n8)
     bAlphaBoolEnabled=Get barrels() -> (barrels:bool)
    ) {blendNode {alphaInputType=Bool, alphaBoolBlend {blendInTime=0.2, blendOption=CircularIn, blendOutTime=0.1}}}) {layerGroup=None}
   Save cached pose 'AnimCache'(Pose=Local Space Ref Pose()) {cacheName=AnimCache}
   n45 = Control Rig(Source=Use cached pose 'AnimCache'()) {
    alphaBoolBlend {blendOption=Linear}
    controlRigAssetReference {blueprintRigClass=&/Prank_flag.Prank_flag_C}
   }
   n46 = RigidBody(ComponentPose=Local To Component(LocalPose=@n45)) {
    bClampLinearTranslationLimitToRefPose=true
    bOverrideWorldGravity=true
    overridePhysicsAsset=&M_MED_Parcel_Prank_Surprise_Pack_Physics_out_Test
    simulationSpace=BaseBoneSpace
    alphaBoolBlend {blendOption=Linear}
    baseBoneRef {boneName=root}
    componentLinearAccScale=<0.5 0.6 0.7>
    componentLinearVelScale=<0.3 0.6 0.5>
    overrideWorldGravity {z=-980}
   }
   n47 = RigidBody(ComponentPose=@n46) {
    bClampLinearTranslationLimitToRefPose=true
    bOverrideWorldGravity=true
    overridePhysicsAsset=&M_MED_Parcel_Prank_Surprise_Pack_Physics_Crank
    alphaBoolBlend {blendOption=Linear}
    componentLinearAccScale=<0.7 0.7 0.7>
    componentLinearVelScale=<0.5 0.5 0.5>
    overrideWorldGravity {z=-980}
   }
   n48 = "Transform (Modify) Bone - Bone: dyn_r_arm"(ComponentPose=@n47, Alpha=1, Translation="0,0,0", Rotation="3,0.000,0.000", Scale="1,1,1") {
    rotationMode=BMM_Additive
    scaleSpace=BCS_BoneSpace
    alphaBoolBlend {blendOption=Linear}
    boneToModify {boneName=dyn_r_arm}
    rotation {pitch=3}
   }
   n49 = "Transform (Modify) Bone - Bone: dyn_l_arm"(ComponentPose=@n48, Alpha=1, Translation="0,0,0", Rotation="-3,0.000,0.000", Scale="1,1,1") {
    rotationMode=BMM_Additive
    scaleSpace=BCS_BoneSpace
    alphaBoolBlend {blendOption=Linear}
    boneToModify {boneName=dyn_l_arm}
    rotation {pitch=-3}
   }
   n50 = "Transform (Modify) Bone - Bone: dyn_spring_MSTR"!(ComponentPose=@n49, Alpha:float=1, Translation="0,0,2", Rotation="0,0,0", Scale="1,1,1") -> (Pose) {translationMode=BMM_Additive, alphaBoolBlend {blendOption=Linear}, boneToModify {boneName=dyn_spring_MSTR}, translation {z=2}}
   n51 = "Transform (Modify) Bone - Bone: dyn_spring_MSTR"!(ComponentPose=@n50, bAlphaBoolEnabled:bool=Get bIsTargeting(), Translation="0,0,0", Rotation="0,-45,0", Scale="1,1,1") -> (Pose) {
    alphaInputType=Bool
    rotationMode=BMM_Additive
    alphaBoolBlend {blendOption=Linear}
    boneToModify {boneName=dyn_spring_MSTR}
    rotation {yaw=-45}
   }
   Two Way Blend(B=Component To Local(ComponentPose=@n51), bAlphaBoolEnabled=Get Flags() -> (Flags:bool)) {blendNode {alphaInputType=Bool, alphaBoolBlend {blendInTime=0.05, blendOption=Cubic, blendOutTime=0.1}}}
   SelectFloat(A=1, B=0, bPickA=false)
   Get Surprise Force()
   Get Surprise Force()
  ]}
 EventGraph {nodes [
   Event OnCharacterCustomizationLoaded()
   > n1 = Cast To FortPlayerPawn(Object=TryGetFortPawn())
   > n2 = WatchCosmeticStat!(
    execute:Exec
    Pawn:Fort Player Pawn Object Reference=@n1
    StatTag:Fort Stat Manager Tag Structure=(TagName="StatManager.AthenaKills")
    DefaultValue:int=0
    bBroadcastInitialValue:bool=false
    InitialValueBroadcastDelay:float=-1
    ChangeValueBroadcastDelay:float=-1
    bConsumeWaitingNotifications:bool=true
   ) -> (then:Exec, OnStatChanged:Exec, StatName:Name, StatValue:int) {is=title K2Node_AsyncAction}
   Set KillCount(execute=@n2.OnStatChanged, KillCount:double="To Float (Integer)"(InInt=@n2.StatValue)) -> (Output_Get:double)
   > Set Active_Kill(Active_Kill:bool=true) -> (Output_Get:bool)
   Event BlueprintInitializeAnimation()
   > Set Skel_Mesh(Skel_Mesh=GetOwnerSkeletalMeshComponent())
   n3 = Event BlueprintUpdateAnimation()
   > Get!(execute:Exec) -> (then:Exec, else:Exec, Skel_Mesh:Skeletal Mesh Component Object Reference)
   > n4 = Sequence!(execute) -> (then_0, then_1, then_2:Exec, then_3:Exec, then_4:Exec, then_5:Exec, then_6:Exec, then_7:Exec, then_8:Exec, then_9:Exec, then_10:Exec)
   > n5 = Branch(Condition=Get Surprise_Out())
   > Branch(Condition=Get Surprise_Out())
   > Set Surprise_Alpha(Surprise_Alpha:double=1) -> (Output_Get:double)
   > n6 = Set Surprise_Timer(Surprise_Timer:double=float + float(A=Get Surprise_Timer() -> (Surprise_Timer:double), B=1) {is=title K2Node_CommutativeAssociativeBinaryOperator}) -> (Output_Get:double)
   > Set Surprise_Scale(Surprise_Scale:Vector=VInterpTo_Constant!(Current:Vector=Get Surprise_Scale(), Target:Vector="1.1,1.1,1.1", DeltaTime:float=@n3.DeltaTimeX, InterpSpeed:float=22) -> (ReturnValue:Vector)) -> (Output_Get:Vector)
   > Set arm_scale(arm_scale:Vector=VInterpTo_Constant(Current=Get arm_scale(), Target="1.1,1.1,1.1", DeltaTime=@n3.DeltaTimeX, InterpSpeed=15)) -> (Output_Get:Vector)
   > Branch(Condition="Equal (Float)"!(A:double=@n6, B:double=Get Surprise_Length_Out() -> (Surprise_Length_Out:double)) -> (ReturnValue:bool) {is=title K2Node_CallFunction})
   > Set Surprise_Out(Surprise_Out:bool=false) -> (Output_Get:bool)
   > Set Surprise_Timer(Surprise_Timer=0)
   > Set Surprise_Alpha(Surprise_Alpha=0)
   > Set Crank_Timer(Crank_Timer:double=0) -> (Output_Get:double)
   > Set Crank_Alpha(Crank_Alpha:double=1) -> (Output_Get:double)
   > Set Physic out(Physic out:bool=false) -> (Output_Get:bool)
   > Set barrels(barrels:bool=false) -> (Output_Get:bool)
   > Set Flags(Flags:bool=false) -> (Output_Get:bool)
   > Set flags 2(flags 2:bool=false) -> (Output_Get:bool)
   > Set physics timer(physics timer:double=0) -> (Output_Get:double)
   > Set barrels timer(barrels timer:double=0) -> (Output_Get:double)
   > Set Flags Timer(Flags Timer:double=0) -> (Output_Get:double)
   > Set Flags 2 timer(Flags 2 timer:double=0) -> (Output_Get:double)
   Set Surprise_Scale(
    execute=@n5.else
    Surprise_Scale=VInterpTo_Constant(Current=Get Surprise_Scale(), Target="0.4,0.5,0.3", DeltaTime=@n3.DeltaTimeX, InterpSpeed=16)
   )
   > Set arm_scale(arm_scale=VInterpTo_Constant(Current=Get arm_scale(), Target="0.1,0.1,0.1", DeltaTime=@n3.DeltaTimeX, InterpSpeed=22))
   > Set Flag Scale(Flag Scale:Vector=VInterpTo_Constant(Current=Get Flag Scale(), Target="0.1,1,0.7", DeltaTime=@n3.DeltaTimeX, InterpSpeed=15)) -> (Output_Get:Vector)
   > Set Flag 2 scale(Flag 2 scale:Vector=VInterpTo_Constant(Current=Get Flag 2 scale(), Target="0.1,1,0.7", DeltaTime=@n3.DeltaTimeX, InterpSpeed=15)) -> (Output_Get:Vector)
   > Set Surprise_Length_Out(Surprise_Length_Out:double=600) -> (Output_Get:double)
   > Set Crank_time_rot(Crank_time_rot:double=RandomFloatInRange(Min=100, Max=400)) -> (Output_Get:double)
   > Set Crank Rot(Crank Rot:Rotator=MakeRotator()) -> (Output_Get:Rotator)
   Branch(execute=@n4.then_10, Condition=Get flags 2() -> (flags 2:bool))
   > Set Flag 2 scale(Flag 2 scale=VInterpTo_Constant(Current=Get Flag 2 scale(), Target="1,1,1", DeltaTime=@n3.DeltaTimeX, InterpSpeed=22))
   n7 = Branch(execute=@n4.then_3, Condition=Get Surprise_Out())
   > Do Once(Reset=@n9.then)
   > n8 = Call Prank FX!(execute:Exec, self:M MED Parcel Prank Surprise Pack Anim BP Object Reference) -> (then:Exec) {is=title K2Node_CallDelegate}
   Do Once(execute=@n7.else, Reset=@n8.then, Start Closed=true)
   > n9 = Call Prank FXOff!(execute:Exec, self:M MED Parcel Prank Surprise Pack Anim BP Object Reference) -> (then:Exec) {is=title K2Node_CallDelegate}
   Branch(execute=@n4.then_8, Condition=Get Flags())
   > Set Flag Scale(Flag Scale=VInterpTo_Constant(Current=Get Flag Scale(), Target="1,1,1", DeltaTime=@n3.DeltaTimeX, InterpSpeed=17))
   Branch(execute=@n4.then_4, Condition=Get Surprise_Out())
   > n10 = Set physics timer(physics timer=float + float(A=Get physics timer() -> (physics timer:double), B=1))
   > Branch(Condition="Equal (Float)"(A=@n10, B=10))
   > Set Physic out(Physic out=true)
   Branch(execute=@n4.then_9, Condition=Get Surprise_Out())
   > n11 = Set Flags 2 timer(Flags 2 timer=float + float(A=Get Flags 2 timer() -> (Flags 2 timer:double), B=1))
   > Branch(Condition="Equal (Float)"(A=@n11, B=28))
   > Set flags 2(flags 2=true)
   n12 = GetGameTimeInSeconds()
   n13 = MapRangeClamped(Value="Cos (Degrees)"!(A:double=float * float(A=@n12, B=75)) -> (ReturnValue:double), InRangeA=-3, InRangeB=3, OutRangeA=-30, OutRangeB=30)
   Set Spring_angle(execute=@n4.then_2, Spring_angle:Rotator=MakeRotator(Roll=@n13, Pitch=@n13)) -> (Output_Get:Rotator)
   Branch(execute=@n4.then_6, Condition=Get Surprise_Out())
   > n14 = Set Flags Timer(Flags Timer=float + float(A=Get Flags Timer() -> (Flags Timer:double), B=1))
   > Branch(Condition="Equal (Float)"(A=@n14, B=25))
   > Set Flags(Flags=true)
   n15 = Branch(execute=@n4.then_7, Condition=Get Surprise_Out())
   n16 = RandomFloatInRangeFromStream!(Stream:"Random Stream Structure (by ref)"=Get random_stream() -> (random_stream:Random Stream Structure), Min:float=-10000, Max:float=10000) -> (ReturnValue:float)
   > Set Surprise Force(Surprise Force:Vector=vector * vector(A:Vector=MakeVector(X=@n16, Y=@n16), B:Vector="0.2,0.2,0.2") -> (ReturnValue:Vector) {is=title K2Node_CommutativeAssociativeBinaryOperator}) -> (Output_Get:Vector)
   Set Surprise Force(execute=@n15.else, Surprise Force="0, 0, 0")
   Branch(execute=@n4.then_1, Condition=Get Active_Kill() -> (Active_Kill:bool))
   n17 = float * float(A=Get Crank_Timer() -> (Crank_Timer:double), B=17)
   > Set Crank Rot(Crank Rot=Invert Rotator!(A:Rotator=MakeRotator(Roll=@n17)) -> (ReturnValue:Rotator))
   > Set Crank_Alpha(Crank_Alpha=1)
   > n18 = Set Crank_Timer(Crank_Timer=float + float(A=Get Crank_Timer(), B=1))
   > Branch(Condition="float >= float"(A=@n18, B=Get Crank_time_rot() -> (Crank_time_rot:double)) {is=title K2Node_CallFunction})
   > Set Active_Kill(Active_Kill=false)
   > Set Surprise_Out(Surprise_Out=true)
   Branch(execute=@n4.then_5, Condition=Get Surprise_Out())
   > n19 = Set barrels timer(barrels timer=float + float(A=Get barrels timer() -> (barrels timer:double), B=1))
   > Branch(Condition="Equal (Float)"(A=@n19, B=15))
   > Set barrels(barrels=true)
   Reroute Node()
   > Reroute Node()
   Reroute Node()
   > Reroute Node()
   GetGameTimeInSeconds()
   "Equal (Float)"(A=0, B=0)
   NormalizeAxis!(Angle:float=0) -> (ReturnValue:float)
   Set random_stream(random_stream:Random Stream Structure=MakeRandomStream!(InitialSeed:int=0) -> (ReturnValue:Random Stream Structure)) -> (Output_Get:Random Stream Structure)
  ]}
 PrankFX {nodes [PrankFX() {is=title K2Node_FunctionEntry}]}
 PrankFXOff {nodes [PrankFXOff() {is=title K2Node_FunctionEntry}]}
]
variables [
 Active_Kill {type=boolean}
 arm_scale {type=Vector}
 barrels {type=boolean}
 barrels timer {type=number}
 Crank Rot {type=Rotator}
 Crank_Alpha {type=number}
 Crank_time_rot {type=number, default=189.6512}
 Crank_Timer {type=number}
 Flag 2 scale {type=Vector}
 Flag Scale {type=Vector}
 Flags {type=boolean}
 flags 2 {type=boolean}
 Flags 2 timer {type=number}
 Flags Timer {type=number}
 KillCount {type=number}
 Physic out {type=boolean}
 physics timer {type=number}
 PrankFX {default=()}
 PrankFXOff {default=()}
 random_stream {type=RandomStream, default {initialSeed=0, seed=0}}
 Skel_Mesh {type=SkeletalMeshComponent}
 Spring_angle {type=Rotator}
 Surprise Force {type=Vector}
 Surprise_Alpha {type=number}
 Surprise_Length_Out {type=number, default=400}
 Surprise_Out {type=boolean}
 Surprise_Scale {type=Vector, default=<0.1 0.1 0.1>}
 Surprise_Timer {type=number}
]
